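; Gameplay ability for the Damage After Honking quest, which credits progress when a qualifying player triggers its gameplay event against a qualifying target.
; On activation it calls the parent, then skips any event whose instigator tags match the exclude container.
; Otherwise, if the instigator's player pawn has all player-container tags and the target tags match the target container, it sends the DamageAfterHonking stat event to the quest manager.
; The ability ends in every case.
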(event Ability|EventActivateAbilityFromEvent
  (params (EventData :type Gameplay%20Event%20Data%20Structure%20%28by%20ref%29))
  (bind (_eventtag _instigator _target _optionalobject _optionalobject2 _contexthandle _instigatortags _targettags) (Utilities|Struct|BreakGameplayEventData EventData))
  (|CallParentFunction EventData)
  (if (GameplayTags|HasAnyTags _instigatortags (Variables|Default|GetExclude%20Container))
    _
    (elif (and (GameplayTags|HasAllMatchingGameplayTags (Game|Player|GetPlayerPawn (Fortnite|GetFortPlayerControllerfromActor _instigator)) (Variables|Default|GetPlayer%20Container)) (GameplayTags|HasAnyTags _targettags (Variables|Default|GetTarget%20Container)))
      (Quest|SendComplexCustomStatEvent (Variables|GAVars|Data|GetOwningPlayerQuestManager) 0 "(GameplayTags=((TagName=\"Athena.Quests.DamageAfterHonking\")))" "(GameplayTags=)" 1)))
  (Ability|EndAbility))
